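# Material function that exposes the post process ambient cubemap intensity to materials as a single scalar output.
# A Custom HLSL node reads View.AmbientCubemapIntensity directly from the view uniform data and feeds it to the AmbientCubemapIntensity output.
# It is listed in the material library under Misc and GetPostProcessSetting.
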MaterialFunction /Engine/Functions/GetPostProcessFunctions/GetAmbientCubemapIntensity
graph {
 properties {bExposeToLibrary=true, description=Get Ambient Cubemap Intensity, libraryCategoriesText [Misc, GetPostProcessSetting]}
 nodes [
  n1 = Custom() {
   code=View.AmbientCubemapIntensity
   desc="\n*** Original source before expression upgrade ***\nView.AmbientCubemapIntensity\n*** Original source before expression upgrade ***\nFrame.AmbientCubemapIntensity"
   description=Ambient Cubemap Intensity
   outputType=CMOT_Float1
   inputs []
  }
  FunctionOutput(None=@n1) {outputName=AmbientCubemapIntensity}
 ]
}
