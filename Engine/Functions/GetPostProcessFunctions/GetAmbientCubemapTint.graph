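# Material function that exposes the view's ambient cubemap tint to materials, listed in the Misc and GetPostProcessSetting library categories.
# A single Custom HLSL node reads View.AmbientCubemapTint as a float4 and feeds it straight to the AmbientCubemapTint output.
# The node's notes record an upgrade from an older Frame.AmbientCubemapTint expression.
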MaterialFunction /Engine/Functions/GetPostProcessFunctions/GetAmbientCubemapTint
graph {
 properties {bExposeToLibrary=true, description=Get Ambient Cubemap Tint, libraryCategoriesText [Misc, GetPostProcessSetting]}
 nodes [
  n1 = Custom() {
   code=View.AmbientCubemapTint
   desc="\n*** Original source before expression upgrade ***\nView.AmbientCubemapTint\n*** Original source before expression upgrade ***\nFrame.AmbientCubemapTint"
   description=Ambient Cubemap Tint
   outputType=CMOT_Float4
   inputs []
  }
  FunctionOutput(None=@n1) {outputName=AmbientCubemapTint}
 ]
}
